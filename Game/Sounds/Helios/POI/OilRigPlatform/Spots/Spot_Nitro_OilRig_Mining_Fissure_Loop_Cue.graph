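# Looping ambient spot cue for a nitro mining fissure at the OilRigPlatform POI.
# Mixes random delayed spark bursts with crackle, two gas-release, earth-destruction and rumble loops, identical across quality levels.
# Each layer has its own attenuation, while looping envelopers drift pitch on the loops and slowly swell destruction, rumble and crackle volume.
# The cue uses the earthquake sound class, but the rumble branch overrides it to SFX.
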SoundCue /Game/Sounds/Helios/POI/OilRigPlatform/Spots/Spot_Nitro_OilRig_Mining_Fissure_Loop_Cue
root=QualityLevel_0
nodes [
 {
  id=Attenuation_0
  children [Mixer_5]
  properties {
   bOverrideAttenuation=true
   attenuationOverrides {
    bAttenuateWithLPF=true
    bEnableListenerFocus=true
    bEnableSubmixSends=true
    falloffDistance=3000
    hPFFrequencyAtMax=100
    lPFFrequencyAtMax=100
    lPFRadiusMax=5000
    lPFRadiusMin=500
    nonFocusVolumeAttenuation=0.2
    occlusionTraceChannel=ECC_GameTraceChannel11
    stereoSpread=10
    attenuationShapeExtents {x=0}
    submixSendSettings [{
      bEnableHPFCutoff=false
      bEnableLPFCutoff=false
      disableManualSendClamp=false
      hPFCutoff=20
      lPFCutoff=20000
      maxSendDistance=3000
      maxSendLevel=0.6
      minSendDistance=500
      minSendLevel=0.4
      sendLevel=1
      sendLevelControlMethod=Linear
      soundSubmix=&Fortnite_Reverb_Distant
      customSendLevelCurve {editorCurveData {keys [<at 0.2 0.4>,<at 1 0.6>]}}
     }]
   }
  }
 }
 {
  id=Attenuation_1
  children [Enveloper_14]
  properties {
   bOverrideAttenuation=true
   attenuationOverrides {
    bAttenuateWithLPF=true
    bEnableListenerFocus=true
    distanceAlgorithm=Logarithmic
    falloffDistance=3500
    hPFFrequencyAtMax=100
    lPFFrequencyAtMax=50
    lPFFrequencyAtMin=8500
    lPFRadiusMax=3400
    lPFRadiusMin=0
    nonFocusVolumeAttenuation=0.2
    occlusionTraceChannel=ECC_GameTraceChannel11
    stereoSpread=0
    attenuationShapeExtents {x=0}
    submixSendSettings [{
      bEnableHPFCutoff=false
      bEnableLPFCutoff=false
      disableManualSendClamp=false
      hPFCutoff=20
      lPFCutoff=20000
      maxSendDistance=3000
      maxSendLevel=1
      minSendDistance=500
      minSendLevel=0.6
      sendLevel=1
      sendLevelControlMethod=Linear
      soundSubmix=&Fortnite_Reverb_Default
      customSendLevelCurve {editorCurveData {keys [<at 0.2 0.6>,<at 1 1>]}}
     }]
   }
  }
 }
 {
  id=Attenuation_2
  children [Enveloper_1]
  properties {
   bOverrideAttenuation=true
   attenuationOverrides {
    bAttenuateWithLPF=true
    bEnableFocusInterpolation=true
    bEnableListenerFocus=true
    bEnableSubmixSends=true
    distanceAlgorithm=Logarithmic
    falloffDistance=1500
    focusReleaseInterpSpeed=0.9
    hPFFrequencyAtMax=100
    lPFFrequencyAtMax=100
    lPFRadiusMax=3500
    lPFRadiusMin=500
    nonFocusVolumeAttenuation=0.1
    attenuationShapeExtents {x=0}
    submixSendSettings [{
      bEnableHPFCutoff=false
      bEnableLPFCutoff=false
      disableManualSendClamp=false
      hPFCutoff=20
      lPFCutoff=20000
      maxSendDistance=3000
      maxSendLevel=1
      minSendDistance=400
      minSendLevel=0.2
      sendLevel=1
      sendLevelControlMethod=Linear
      soundSubmix=&Fortnite_Reverb_Default
      customSendLevelCurve {editorCurveData {keys [<at 0.2 0.2>,<at 1 1>]}}
     }]
   }
  }
 }
 {
  id=Attenuation_4
  children [Enveloper_5]
  properties {
   bOverrideAttenuation=true
   attenuationOverrides {
    bAttenuateWithLPF=true
    bEnableListenerFocus=true
    bEnableSubmixSends=true
    distanceAlgorithm=Logarithmic
    falloffDistance=2500
    hPFFrequencyAtMax=100
    lPFFrequencyAtMax=100
    lPFRadiusMax=7500
    lPFRadiusMin=500
    nonFocusVolumeAttenuation=0.4
    occlusionInterpolationTime=0.5
    occlusionLowPassFilterFrequency=8500
    occlusionTraceChannel=ECC_GameTraceChannel11
    stereoSpread=100
    attenuationShapeExtents {x=500}
    submixSendSettings [{
      bEnableHPFCutoff=false
      bEnableLPFCutoff=false
      disableManualSendClamp=false
      hPFCutoff=20
      lPFCutoff=20000
      maxSendDistance=3000
      maxSendLevel=1
      minSendDistance=400
      minSendLevel=0.2
      sendLevel=1
      sendLevelControlMethod=Linear
      soundSubmix=&Fortnite_Reverb_Default
      customSendLevelCurve {editorCurveData {keys [<at 0.2 0.2>,<at 1 1>]}}
     }]
   }
  }
 }
 {
  id=Attenuation_5
  children [Enveloper_8]
  properties {
   bOverrideAttenuation=true
   attenuationOverrides {
    bAttenuateWithLPF=true
    bEnableListenerFocus=true
    bEnableSubmixSends=true
    distanceAlgorithm=Logarithmic
    falloffDistance=3500
    hPFFrequencyAtMax=100
    lPFFrequencyAtMax=1000
    lPFRadiusMax=4000
    lPFRadiusMin=1000
    nonFocusVolumeAttenuation=0.4
    nonSpatializedRadiusEnd=750
    nonSpatializedRadiusStart=250
    occlusionLowPassFilterFrequency=2500
    occlusionTraceChannel=ECC_GameTraceChannel11
    stereoSpread=50
    attenuationShapeExtents {x=0}
    submixSendSettings [{
      bEnableHPFCutoff=false
      bEnableLPFCutoff=false
      disableManualSendClamp=false
      hPFCutoff=20
      lPFCutoff=20000
      maxSendDistance=3000
      maxSendLevel=1
      minSendDistance=500
      minSendLevel=0.2
      sendLevel=1
      sendLevelControlMethod=Linear
      soundSubmix=&Fortnite_Reverb_Default
      customSendLevelCurve {editorCurveData {keys [<at 0.2 0.2>,<at 1 1>]}}
     }]
   }
  }
 }
 {
  id=Attenuation_6
  children [Enveloper_10]
  properties {
   bOverrideAttenuation=true
   attenuationOverrides {
    bAttenuateWithLPF=true
    bEnableListenerFocus=true
    bEnableSubmixSends=true
    falloffDistance=3000
    hPFFrequencyAtMax=100
    hPFFrequencyAtMin=80
    lPFRadiusMax=5000
    lPFRadiusMin=500
    nonFocusVolumeAttenuation=0.75
    attenuationShapeExtents {x=1500}
   }
  }
 }
 {id=Delay_0, children [Random_0], properties {delayMax=1}}
 {id=Delay_1, children [Random_1], properties {delayMax=1.5, delayMin=0.5}}
 {
  id=Enveloper_1
  children [Modulator_0]
  properties {
   bLoop=true
   bLoopIndefinitely=true
   loopEnd=30
   pitchCurve {editorCurveData {keys [{interpMode=RCIM_Cubic, time=0}, <at 15.9468 0.9 RCIM_Cubic>, {interpMode=RCIM_Cubic, time=30}]}}
   volumeCurve {editorCurveData {keys [<at 0 1>]}}
  }
 }
 {
  id=Enveloper_10
  children [Enveloper_9]
  properties {
   bLoop=true
   bLoopIndefinitely=true
   loopEnd=120
   pitchCurve {editorCurveData {keys [{time=0}]}}
   volumeCurve {editorCurveData {keys [
      {arriveTangent=0.0001, interpMode=RCIM_Cubic, leaveTangent=0.0082, tangentMode=RCTM_User, time=0.5, value=0.3803}
      {arriveTangent=0.0318, interpMode=RCIM_Cubic, leaveTangent=0.0318, time=10, value=0.6}
      {arriveTangent=-0.0028, interpMode=RCIM_Cubic, leaveTangent=-0.0028, time=20, value=1}
      <at 45.8196 0.5 RCIM_Cubic>
      {arriveTangent=0.0151, interpMode=RCIM_Cubic, leaveTangent=0.0151, time=60, value=1}
      {arriveTangent=-0.0202, interpMode=RCIM_Cubic, leaveTangent=-0.0202, tangentMode=RCTM_User, time=82.1697, value=1.05}
      <at 120 0.5>
     ]}}
  }
 }
 {
  id=Enveloper_13
  children [Modulator_6]
  properties {
   bLoop=true
   bLoopIndefinitely=true
   loopEnd=30
   pitchCurve {editorCurveData {keys [{interpMode=RCIM_Cubic, time=0}, <at 15 1.1 RCIM_Cubic>, {interpMode=RCIM_Cubic, time=30}]}}
   volumeCurve {editorCurveData {keys [<at 0 1>]}}
  }
 }
 {
  id=Enveloper_14
  children [Enveloper_13]
  properties {
   bLoop=true
   bLoopIndefinitely=true
   loopEnd=30
   pitchCurve {editorCurveData {keys [{time=0}]}}
   volumeCurve {editorCurveData {keys [
      {arriveTangent=0.0001, interpMode=RCIM_Cubic, time=0, value=1}
      <at 2.5 1.1 RCIM_Cubic>
      {arriveTangent=0.0057, interpMode=RCIM_Cubic, leaveTangent=0.0057, time=15, value=1}
      <at 17.5 1.1 RCIM_Cubic>
      <at 30 1 RCIM_Cubic>
     ]}}
  }
 }
 {
  id=Enveloper_5
  children [Modulator_1]
  properties {
   bLoop=true
   bLoopIndefinitely=true
   loopEnd=30
   pitchCurve {editorCurveData {keys [{interpMode=RCIM_Cubic, time=0}, <at 15.9468 0.9 RCIM_Cubic>, {interpMode=RCIM_Cubic, time=30}]}}
   volumeCurve {editorCurveData {keys [<at 0 1>]}}
  }
 }
 {
  id=Enveloper_7
  children [Modulator_3]
  properties {
   bLoop=true
   bLoopIndefinitely=true
   loopEnd=30
   pitchCurve {editorCurveData {keys [{interpMode=RCIM_Cubic, time=0}, <at 15.9468 0.9 RCIM_Cubic>, {interpMode=RCIM_Cubic, time=30}]}}
   volumeCurve {editorCurveData {keys [<at 0 1>]}}
  }
 }
 {
  id=Enveloper_8
  children [Enveloper_7]
  properties {
   bLoop=true
   bLoopIndefinitely=true
   loopEnd=120
   pitchCurve {editorCurveData {keys [{time=0}]}}
   volumeCurve {editorCurveData {keys [
      {arriveTangent=0.0001, interpMode=RCIM_Cubic, leaveTangent=0.0082, tangentMode=RCTM_User, time=0.5, value=0.3803}
      {arriveTangent=0.0318, interpMode=RCIM_Cubic, leaveTangent=0.0318, time=10, value=0.6}
      {arriveTangent=-0.0028, interpMode=RCIM_Cubic, leaveTangent=-0.0028, time=20, value=1}
      <at 45.8196 0.5 RCIM_Cubic>
      {arriveTangent=0.0151, interpMode=RCIM_Cubic, leaveTangent=0.0151, time=60, value=1}
      {arriveTangent=-0.0202, interpMode=RCIM_Cubic, leaveTangent=-0.0202, tangentMode=RCTM_User, time=82.1697, value=1.05}
      <at 120 0.5>
     ]}}
  }
 }
 {
  id=Enveloper_9
  children [Modulator_4]
  properties {
   bLoop=true
   bLoopIndefinitely=true
   loopEnd=30
   pitchCurve {editorCurveData {keys [{interpMode=RCIM_Cubic, time=0}, <at 15.9468 0.9 RCIM_Cubic>, {interpMode=RCIM_Cubic, time=30}]}}
   volumeCurve {editorCurveData {keys [<at 0 1>]}}
  }
 }
 {id=Looping_0, children [Delay_0]}
 {id=Looping_1, children [Delay_1]}
 {id=Mixer_0, children [WavePlayer_1], properties {inputVolume [0.5]}}
 {id=Mixer_1, children [WavePlayer_2], properties {inputVolume [1]}}
 {
  id=Mixer_2
  children [Attenuation_0, Attenuation_1, Attenuation_2, Attenuation_4, Attenuation_5, SoundClass_0]
  properties {inputVolume [0.4,0.3,1,1.5,1.5,1]}
 }
 {id=Mixer_3, children [WavePlayer_17], properties {inputVolume [1]}}
 {id=Mixer_4, children [WavePlayer_0], properties {inputVolume [1]}}
 {id=Mixer_5, children [Looping_1,Looping_0], properties {inputVolume [1,1]}}
 {id=Mixer_6, children [WavePlayer_11], properties {inputVolume [1]}}
 {id=Modulator_0, children [Mixer_1]}
 {id=Modulator_1, children [Mixer_0], properties {pitchMax=0.8, pitchMin=0.8}}
 {id=Modulator_3, children [Mixer_3]}
 {id=Modulator_4, children [Mixer_4]}
 {id=Modulator_6, children [Mixer_6]}
 {id=QualityLevel_0, children [Mixer_2,Mixer_2,Mixer_2,Mixer_2]}
 {id=Random_0, children [WavePlayer_12, WavePlayer_13, WavePlayer_14, WavePlayer_15, WavePlayer_16], properties {weights [1,1,1,1,1]}}
 {id=Random_1, children [WavePlayer_3,WavePlayer_4,WavePlayer_5,WavePlayer_6,WavePlayer_7], properties {weights [1,1,1,1,1]}}
 {id=SoundClass_0, children [Attenuation_6], properties {soundClassOverride=&SFX}}
 {id=WavePlayer_0, properties {bLooping=true, soundWaveAssetPtr=&/Spot_Nitro_Earth_Rumble_Loop}}
 {
  id=WavePlayer_1
  properties {bLooping=true, soundWaveAssetPtr=&/Spot_Nitro_Gas_Release_Close_Loop}
 }
 {id=WavePlayer_11, properties {bLooping=true, soundWaveAssetPtr=&/Spot_Nitro_Spark_Crackle_Loop}}
 {id=WavePlayer_12, properties {soundWaveAssetPtr=&/Spot_Nitro_Spark_Crackle_Burst_01}}
 {id=WavePlayer_13, properties {soundWaveAssetPtr=&/Spot_Nitro_Spark_Crackle_Burst_02}}
 {id=WavePlayer_14, properties {soundWaveAssetPtr=&/Spot_Nitro_Spark_Crackle_Burst_03}}
 {id=WavePlayer_15, properties {soundWaveAssetPtr=&/Spot_Nitro_Spark_Crackle_Burst_04}}
 {id=WavePlayer_16, properties {soundWaveAssetPtr=&/Spot_Nitro_Spark_Crackle_Burst_05}}
 {
  id=WavePlayer_17
  properties {bLooping=true, soundWaveAssetPtr=&/Spot_NitroMine_Earth_Destruction_Loop}
 }
 {id=WavePlayer_2, properties {bLooping=true, soundWaveAssetPtr=&/Spot_Nitro_Gas_Release_Loop}}
 {id=WavePlayer_3, properties {soundWaveAssetPtr=&Spot_Nitro_Spark_Crackle_Burst_01}}
 {id=WavePlayer_4, properties {soundWaveAssetPtr=&Spot_Nitro_Spark_Crackle_Burst_02}}
 {id=WavePlayer_5, properties {soundWaveAssetPtr=&Spot_Nitro_Spark_Crackle_Burst_03}}
 {id=WavePlayer_6, properties {soundWaveAssetPtr=&Spot_Nitro_Spark_Crackle_Burst_04}}
 {id=WavePlayer_7, properties {soundWaveAssetPtr=&Spot_Nitro_Spark_Crackle_Burst_05}}
]
properties {
 bOverrideConcurrency=true
 maxDistance=4500
 soundClassObject=/Game/Sounds/FortSoundClassesAndModes/Fort_Earthquakes
 volumeMultiplier=0.9
 attenuationOverrides {
  attenuationShape=Capsule
  bEnableFocusInterpolation=true
  bEnableListenerFocus=true
  falloffDistance=1200
  focusAzimuth=15
  focusReleaseInterpSpeed=0.9
  nonFocusAzimuth=45
  nonFocusVolumeAttenuation=0.4
  nonSpatializedRadiusEnd=125
  nonSpatializedRadiusStart=150
  stereoSpread=150
  attenuationShapeExtents=<1200 350>
 }
 busSends [{audioBus=null, minSendLevel=0.6, sendLevel=1, soundSourceBus=&AmbientAudio_Exterior_SourceBus}]
 concurrencyOverrides {maxCount=2, voiceStealReleaseTime=1}
 concurrencySet [&Weapon_Drop_StopOldest]
}
