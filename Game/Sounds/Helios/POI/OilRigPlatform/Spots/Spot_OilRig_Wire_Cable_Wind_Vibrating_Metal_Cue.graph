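# Spot ambience cue for wind-vibrated wire cable and metal on the oil rig POI, limited to two concurrent instances.
# A fading-in vibrating-metal loop uses nested looping envelopes for slow volume and pitch drift.
# Two randomly delayed loops mix in pitch-modulated cable creaks and quieter metal stresses, reusing Artemis waves.
# Creaks randomly pick fade-out, plain or fade-in variants and play only in the first two of four quality slots.
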SoundCue /Game/Sounds/Helios/POI/OilRigPlatform/Spots/Spot_OilRig_Wire_Cable_Wind_Vibrating_Metal_Cue
root=Mixer_0
nodes [
 {id=Delay_1, children [QualityLevel_1], properties {delayMax=3, delayMin=1}}
 {id=Delay_2, children [Modulator_2], properties {delayMax=3, delayMin=1}}
 {
  id=Enveloper_0
  children [Modulator_1]
  properties {
   bLoop=true
   bLoopIndefinitely=true
   loopEnd=30
   pitchCurve {editorCurveData {keys [{interpMode=RCIM_Cubic, time=0}, <at 15.9468 0.9 RCIM_Cubic>, {interpMode=RCIM_Cubic, time=30}]}}
   volumeCurve {editorCurveData {keys [<at 0 1>]}}
  }
 }
 {
  id=Enveloper_1
  children [Modulator_0]
  properties {
   pitchCurve {editorCurveData {keys [{time=0}]}}
   volumeCurve {editorCurveData {keys [<at 0 1>, {arriveTangent=-0.2, interpMode=RCIM_Cubic, leaveTangent=-0.2, time=3, value=1}, <at 5 0>]}}
  }
 }
 {
  id=Enveloper_2
  children [Modulator_0]
  properties {
   pitchCurve {editorCurveData {keys [{time=0}]}}
   volumeCurve {editorCurveData {keys [<at 0 0 RCIM_Cubic>, {arriveTangent=0.0151, interpMode=RCIM_Cubic, leaveTangent=0.0151, tangentMode=RCTM_User, time=3, value=0}, <at 5 1>]}}
  }
 }
 {
  id=Enveloper_29
  children [Enveloper_3]
  properties {
   pitchCurve {editorCurveData {keys [{time=0}]}}
   volumeCurve {editorCurveData {keys [{arriveTangent=3.4293, interpMode=RCIM_Cubic, leaveTangent=3.4293, tangentMode=RCTM_User, time=0, value=0}, <at 0.4 1 RCIM_Cubic>]}}
  }
 }
 {
  id=Enveloper_3
  children [Enveloper_0]
  properties {
   bLoop=true
   bLoopIndefinitely=true
   loopEnd=120
   pitchCurve {editorCurveData {keys [{time=0}]}}
   volumeCurve {editorCurveData {keys [
      {arriveTangent=0.0001, interpMode=RCIM_Cubic, leaveTangent=0.0082, tangentMode=RCTM_User, time=0.5, value=0.3803}
      {arriveTangent=0.0318, interpMode=RCIM_Cubic, leaveTangent=0.0318, time=10, value=0.6}
      {arriveTangent=-0.0028, interpMode=RCIM_Cubic, leaveTangent=-0.0028, time=20, value=1}
      <at 45.8196 0.5 RCIM_Cubic>
      {arriveTangent=0.0151, interpMode=RCIM_Cubic, leaveTangent=0.0151, time=60, value=1}
      {arriveTangent=-0.0202, interpMode=RCIM_Cubic, leaveTangent=-0.0202, tangentMode=RCTM_User, time=82.1697, value=1.05}
      <at 120 0.5>
     ]}}
  }
 }
 {id=Looping_0, children [Delay_1]}
 {id=Looping_1, children [Delay_2]}
 {id=Mixer_0, children [QualityLevel_0, Looping_0, Looping_1], properties {inputVolume [1,1,0.25]}}
 {id=Modulator_0, children [Random_1], properties {pitchMax=0.9, pitchMin=0.7}}
 {id=Modulator_1, children [WavePlayer_1], properties {pitchMax=1.2, pitchMin=1}}
 {id=Modulator_2, children [Random_3], properties {pitchMin=0.9}}
 {id=QualityLevel_0, children [Enveloper_29,Enveloper_29,Enveloper_29,Enveloper_29]}
 {id=QualityLevel_1, children [Random_2,Random_2,null,null]}
 {
  id=Random_1
  children [WavePlayer_2, WavePlayer_3, WavePlayer_4, WavePlayer_5, WavePlayer_6, WavePlayer_7, WavePlayer_8, WavePlayer_9, WavePlayer_10]
  properties {weights [1,1,1,1,1,1,1,1,1]}
 }
 {id=Random_2, children [Enveloper_1,Modulator_0,Enveloper_2], properties {weights [1,1,1]}}
 {
  id=Random_3
  children [
   WavePlayer_11
   WavePlayer_12
   WavePlayer_13
   WavePlayer_14
   WavePlayer_15
   WavePlayer_16
   WavePlayer_17
   WavePlayer_18
   WavePlayer_19
   WavePlayer_20
  ]
  properties {weights [1,1,1,1,1,1,1,1,1,1]}
 }
 {
  id=WavePlayer_1
  properties {bLooping=true, soundWaveAssetPtr=/Game/Sounds/Artemis/POI/Blimps_Crashsites/Spots/Spot_Wind_Excited_Metal_Vibrating_Loop}
 }
 {
  id=WavePlayer_10
  properties {soundWaveAssetPtr=/Game/Sounds/Artemis/POI/Blimps_Crashsites/Spots/Spot_Blimp_Crashsite_Cable_Tension_Creak_08}
 }
 {id=WavePlayer_11, properties {soundWaveAssetPtr=/Game/Sounds/Artemis/POI/Tilted_Towers/Spots/Spot_Tilted_Towers_Metal_Stress_Light_01}}
 {id=WavePlayer_12, properties {soundWaveAssetPtr=/Game/Sounds/Artemis/POI/Tilted_Towers/Spots/Spot_Tilted_Towers_Metal_Stress_Light_02}}
 {id=WavePlayer_13, properties {soundWaveAssetPtr=/Game/Sounds/Artemis/POI/Tilted_Towers/Spots/Spot_Tilted_Towers_Metal_Stress_Light_03}}
 {id=WavePlayer_14, properties {soundWaveAssetPtr=/Game/Sounds/Artemis/POI/Tilted_Towers/Spots/Spot_Tilted_Towers_Metal_Stress_Light_04}}
 {id=WavePlayer_15, properties {soundWaveAssetPtr=/Game/Sounds/Artemis/POI/Tilted_Towers/Spots/Spot_Tilted_Towers_Metal_Stress_Light_05}}
 {id=WavePlayer_16, properties {soundWaveAssetPtr=/Game/Sounds/Artemis/POI/Tilted_Towers/Spots/Spot_Tilted_Towers_Metal_Stress_Light_06}}
 {id=WavePlayer_17, properties {soundWaveAssetPtr=/Game/Sounds/Artemis/POI/Tilted_Towers/Spots/Spot_Tilted_Towers_Metal_Stress_Light_07}}
 {id=WavePlayer_18, properties {soundWaveAssetPtr=/Game/Sounds/Artemis/POI/Tilted_Towers/Spots/Spot_Tilted_Towers_Metal_Stress_Light_08}}
 {id=WavePlayer_19, properties {soundWaveAssetPtr=/Game/Sounds/Artemis/POI/Tilted_Towers/Spots/Spot_Tilted_Towers_Metal_Stress_Light_09}}
 {id=WavePlayer_2, properties {soundWaveAssetPtr=/Game/Sounds/Artemis/POI/Blimps_Crashsites/Spots/Spot_Blimp_Crashsite_Cable_Tension_Creak_09}}
 {id=WavePlayer_20, properties {soundWaveAssetPtr=/Game/Sounds/Artemis/POI/Tilted_Towers/Spots/Spot_Tilted_Towers_Metal_Stress_Light_10}}
 {id=WavePlayer_3, properties {soundWaveAssetPtr=/Game/Sounds/Artemis/POI/Blimps_Crashsites/Spots/Spot_Blimp_Crashsite_Cable_Tension_Creak_01}}
 {id=WavePlayer_4, properties {soundWaveAssetPtr=/Game/Sounds/Artemis/POI/Blimps_Crashsites/Spots/Spot_Blimp_Crashsite_Cable_Tension_Creak_02}}
 {id=WavePlayer_5, properties {soundWaveAssetPtr=/Game/Sounds/Artemis/POI/Blimps_Crashsites/Spots/Spot_Blimp_Crashsite_Cable_Tension_Creak_03}}
 {id=WavePlayer_6, properties {soundWaveAssetPtr=/Game/Sounds/Artemis/POI/Blimps_Crashsites/Spots/Spot_Blimp_Crashsite_Cable_Tension_Creak_04}}
 {id=WavePlayer_7, properties {soundWaveAssetPtr=/Game/Sounds/Artemis/POI/Blimps_Crashsites/Spots/Spot_Blimp_Crashsite_Cable_Tension_Creak_05}}
 {id=WavePlayer_8, properties {soundWaveAssetPtr=/Game/Sounds/Artemis/POI/Blimps_Crashsites/Spots/Spot_Blimp_Crashsite_Cable_Tension_Creak_06}}
 {id=WavePlayer_9, properties {soundWaveAssetPtr=/Game/Sounds/Artemis/POI/Blimps_Crashsites/Spots/Spot_Blimp_Crashsite_Cable_Tension_Creak_07}}
]
properties {
 attenuationSettings=&/Spot_OilRig_Wire_Cable_Wind_Vibrating_Metal_Att
 bEnableBusSends=false
 bEnableSubmixSends=false
 bOverrideConcurrency=true
 maxDistance=2003.4948
 soundClassObject=&Fort_Gadgets
 volumeMultiplier=0.6
 concurrencyOverrides {maxCount=2, voiceStealReleaseTime=2}
}
