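# Looping ambient spot sound cue for a large scrap heap fan, filed under the Helios oil rig platform POI.
# A mixer layers two looping waves, an air duct fan and a louder-weighted outpost wall fan spin, through a root modulator.
# It borrows the air duct fan attenuation, plays on the ambient props sound class at reduced volume, and overrides concurrency to prevent new instances at the limit.
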SoundCue /Game/Sounds/Helios/POI/OilRigPlatform/Spots/Spot_ScrapHeap_Fan_Large_Loop_Cue
root=Modulator_0
nodes [
 {id=Mixer_0, children [WavePlayer_2,WavePlayer_1], properties {inputVolume [0.5,2.5]}}
 {id=Modulator_0, children [Mixer_0]}
 {
  id=WavePlayer_1
  properties {bLooping=true, soundWaveAssetPtr=/Game/Sounds/Artemis/Biomes/Grasslands/Grassland_Spots/Spot_Outpost_Wall_Fan_Spin_Loop}
 }
 {id=WavePlayer_2, properties {bLooping=true, soundWaveAssetPtr=/Game/Sounds/Apollo/Props/AirDucts/AMB_Air_Duct_Fan_Loop_01}}
]
properties {
 attenuationSettings=/Game/Sounds/Apollo/Props/AirDucts/AirDuct_Fan_Att
 bOverrideConcurrency=true
 maxDistance=512
 soundClassObject=&Fort_Ambient_Props_Athena
 volumeMultiplier=0.501
 concurrencyOverrides {resolutionRule=PreventNew}
}
