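# Looping sound cue for a small fan ambience spot, named for the scrap heap at the Oil Rig Platform POI.
# A single wave player loops an air duct fan recording through a modulator with fixed unit volume, at roughly half cue volume.
# It borrows the Grasslands outpost wall fan attenuation and uses the Athena ambient props sound class.
# The Athena prop concurrency set applies, alongside a concurrency override with a max count of two.
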SoundCue /Game/Sounds/Helios/POI/OilRigPlatform/Spots/Spot_ScrapHeap_Fan_Small_Loop_Cue
root=Modulator_0
nodes [
 {id=Modulator_0, children [WavePlayer_1], properties {volumeMax=1, volumeMin=1}}
 {id=WavePlayer_1, properties {bLooping=true, soundWaveAssetPtr=/Game/Sounds/Apollo/Props/AirDucts/AMB_Air_Duct_Fan_Loop_01}}
]
properties {
 attenuationSettings=/Game/Sounds/Artemis/Biomes/Grasslands/Grassland_Spots/Spot_Outpost_Wall_Fan_Att
 bEnableBusSends=false
 maxDistance=3950
 soundClassObject=&Fort_Ambient_Props_Athena
 volumeMultiplier=0.501
 concurrencyOverrides {maxCount=2}
 concurrencySet [/Game/Sounds/Concurrency_Rules/Athena_Prop]
 preEffectBusSends [{}]
}
